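; Neon City cyber dragon hologram Blueprint that triggers fire-breath effects and audio when the storm starts holding or shrinking.
; Once the game state is ready, cosmetic-capable instances listen for stateful standard and Creative storm-shield phase-step events.
; Storm steps activate the breath Niagara effect, play a burst and spawn an attached fire loop, which a timer later fades out.
; EndPlay stops both listeners and fades remaining breath audio.
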(event EventBeginPlay
  (Class|FortAsyncActionGameStateReady|GameStateReadyAsync
    (:OnReady
      (Networking|CanExecuteCosmeticEvents
        (:True
          (bind _returnvalue (GameplayEventRouter|GetEventRouter "/Script/FortniteGame.FortGameState" self))
          (bind (_asynctaskproxy _payload) (GameplayEventRouter|StartListeningToStatefulEvent _returnvalue "/Script/FortniteGame.FortGamePhaseStepUpdatedEvent")
            (:then
              (Variables|Default|SetGamePhaseStepUpdatedAsyncTask _asynctaskproxy)
              (bind (_asynctaskproxy_2 _payload_2) (GameplayEventRouter|StartListeningToStatefulEvent _returnvalue "/Script/FortniteGame.FortCreativeStormShieldGamePhaseStepUpdatedEvent")
                (:then
                  (Variables|Default|SetCreativeStormShieldGamePhaseStepUpdatedAsyncTask _asynctaskproxy_2))
                (:OnEventReceived
                  _)
                (:OnSavedEventStateExists
                  _))
              (CallFunction|GamePhaseStepChangedEvent (.GamePhaseStep (Utilities|Struct|BreakFortCreativeStormShieldGamePhaseStepUpdatedEvent _payload_2))))
            (:OnEventReceived
              _)
            (:OnSavedEventStateExists
              _))
          (CallFunction|GamePhaseStepChangedEvent (Utilities|Struct|BreakFortGamePhaseStepUpdatedEvent _payload)))))))

(event Custom|GamePhaseStepChanged_Event
  (params (GamePhaseStep :type EAthenaGamePhaseStep%20Enum))
  (bind _print_to_screen (Variables|Default|GetPrint%20to%20Screen))
  (switch Utilities|FlowControl|Switch|SwitchonEAthenaGamePhaseStep GamePhaseStep
    (:StormHolding
      (Development|PrintString "Dragon BP - Storm Forming" _print_to_screen _print_to_screen)
      (Components|Activation|Activate (Variables|Default|GetNS_NeonCity_CyberDragon_Breath) false)
      (CallFunction|PlaySound))
    (:StormShrinking
      (Development|PrintString "Dragon BP - Storm Shrinking" _print_to_screen _print_to_screen)
      (Components|Activation|Activate (Variables|Default|GetNS_NeonCity_CyberDragon_Breath) false)
      (CallFunction|PlaySound))))

(event Custom|PlaySound
  (bind _returnvalue (Audio|SpawnSoundAttached "/Game/Sounds/Asteria/POI/NeonCity/Holograms/NeonCity_Dragon_Hologram_Fire_Loop_Cue.NeonCity_Dragon_Hologram_Fire_Loop_Cue" (Variables|Default|GetNS_NeonCity_CyberDragon_Breath)))
  (Variables|Default|SetBreathAudio _returnvalue)
  (Audio|Components|Audio|Play (Variables|Default|GetBurstAudio) 0.0)
  (Utilities|Time|SetTimerbyEvent (delegate Custom|StopBreathAudio) 3.5))

(event Custom|StopBreathAudio
  (bind _breathaudio (Variables|Default|GetBreathAudio)
    (:then
      (Audio|Components|Audio|FadeOut _breathaudio 2.0 0.0 "Linear" :!class /Script/Engine.AudioComponent))
    (:else)))

(event EventEndPlay
  (params (EndPlayReason :type EEndPlayReason%20Enum :default "Destroyed"))
  (bind _gamephasestepupdatedasynctask (Variables|Default|GetGamePhaseStepUpdatedAsyncTask)
    (:then
      (GameplayEventRouter|StopListeningtoStatefulEvent _gamephasestepupdatedasynctask))
    (:else
      _))
  (bind _creativestormshieldgamephasestepupdatedasynctask (Variables|Default|GetCreativeStormShieldGamePhaseStepUpdatedAsyncTask)
    (:then
      (GameplayEventRouter|StopListeningtoStatefulEvent _creativestormshieldgamephasestepupdatedasynctask))
    (:else
      _))
  (bind _breathaudio (Variables|Default|GetBreathAudio)
    (:then
      (Audio|Components|Audio|FadeOut _breathaudio 2.0 0.0 "Linear" :!class /Script/Engine.AudioComponent))
    (:else)))

(fn ConstructionScript
  (|CallParentFunction))
