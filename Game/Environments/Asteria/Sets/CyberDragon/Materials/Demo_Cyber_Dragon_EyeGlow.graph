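# Material for the Cyber Dragon demo set in Asteria, named as the glow for the dragon's eyes.
# A single Multiply scales a constant magenta-pink color by a high scalar, giving an overbright HDR value suited to glow.
# It has no parameters or textures, uses Default Lit shading, and is flagged for Nanite and instanced static meshes.
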Material /Game/Environments/Asteria/Sets/CyberDragon/Materials/Demo_Cyber_Dragon_EyeGlow
graph {
 properties {bUsedWithInstancedStaticMeshes=true, bUsedWithNanite=true, pixelDepthOffsetMode=PDOM_Legacy, usedShadingModels=Default Lit}
 nodes [Multiply(A=Constant3Vector() {constant=<rgba 1 0 0.7839 1>}, B=Constant() {r=25})]
}
