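# Parent material for the CyberDragon environment set, a lit Metal-physics surface supporting Nanite and instanced meshes.
# Packed masks pass through overlay adjustments for specular, metallic and roughness, alongside diffuse, normal and emissive textures.
# Distance-field AO, broken up by world-aligned stain-mask channels, provides grime-like occlusion, with a scalable diffuse-darkening blend.
# The AO falls back to 1 on ES3_1 and on Low, Medium and High quality.
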Material /Game/Environments/Asteria/Sets/CyberDragon/Materials/M_CyberDragon_Parent
graph {
 properties {
  bUsedWithInstancedStaticMeshes=true
  bUsedWithNanite=true
  physMaterial=/Game/FortressPhysicalMaterials/Metal
  pixelDepthOffsetMode=PDOM_Legacy
  usedShadingModels=Default Lit
 }
 nodes [
  n1 = TextureSampleParameter2D() {
   parameterName=Masks
   samplerType=SAMPLERTYPE_Masks
   texture=/Game/Global/Textures/Default/Blanks/T_MidgreyMask
   channelNames {a=Unused, b=Roughness, g=Metallic, r=Specular}
  }
  MaterialFunctionCall(Base=@n1.Specular, Blend=ScalarParameter() {defaultValue=0.5, parameterName=SpecularOverlay}) {desc=Specular, materialFunction=&Blend_Overlay}
  MaterialFunctionCall(Base=@n1.Metallic, Blend=ScalarParameter() {defaultValue=0.5, parameterName=MetallicOverlay}) {desc=Metallic, materialFunction=&Blend_Overlay}
  MaterialFunctionCall(Base=@n1.Roughness, Blend=ScalarParameter() {defaultValue=-0.1, parameterName=RoughnessOverlay}) {desc=Roughness, materialFunction=&Blend_Overlay}
  n2 = NamedRerouteUsage()
  n3 = ScalarParameter() {defaultValue=0.5, group=Ambient Occlusion, parameterName=AODiffuseScale}
  Multiply(
   A=VectorParameter() {parameterName=BaseColorScale, defaultValue=<rgba 0.5 0.5 0.5 0>}.RGB
   B=Multiply(
    A=TextureSampleParameter2D() {parameterName=Diffuse, samplerType=SAMPLERTYPE_Masks, texture=&T_MidgreyMask}.RGB
    B=LinearInterpolate(B=@n2, Alpha=@n3) {constA=1}
   )
  )
  n4 = DistanceFieldApproxAO!(
   World Position
   Normal
   BaseDistance=ScalarParameter() {defaultValue=40, group=Ambient Occlusion, parameterName=BaseDistance}
   Radius=ScalarParameter() {defaultValue=50, group=Ambient Occlusion, parameterName=Radius}
  ) {numSteps=2}
  n5 = ScalarParameter() {defaultValue=0.3, group=Ambient Occlusion, parameterName=AOEdgeExponent}
  n6 = Power(
   Base=Min(A=NamedRerouteUsage(), B=NamedRerouteUsage())
   Exp=ScalarParameter() {defaultValue=2, group=Ambient Occlusion, parameterName=AONoiseExp}
  )
  n7 = LinearInterpolate(
   B=Subtract(A=Power(Base=@n4, Exp=@n5), B=LinearInterpolate(A=@n6, Alpha=@n4) {constB=0})
   Alpha=ScalarParameter() {defaultValue=1, group=Ambient Occlusion, parameterName=AO_FinalStrength}
  ) {constA=1}
  n8 = Constant() {r=1}
  NamedRerouteDeclaration(None=FeatureLevelSwitch(Default=QualitySwitch(Default=@n7, Low=@n8, High=@n8, Medium=@n8), ES3_1=@n8)) {name=Ambient Occlusion, nodeColor=<rgba 0 0.5312 1 1>}
  n9 = TextureObjectParameter() {parameterName=StainMap, samplerType=SAMPLERTYPE_Masks, texture=/Game/Environments/Asteria/Sets/NeonCity/Techart/Textures/T_StainA_Masks}
  n10 = MaterialFunctionCall(
   TextureObject=@n9
   TextureSize=ScalarParameter() {defaultValue=800, group=Ambient Occlusion, parameterName=AOWorldAlignSize}
   Export Float 4=StaticBool() {value=true}
  ) {materialFunction=&WorldAlignedTexture}
  n11 = ChannelMaskParameter(None=@n10.XYZ Texture) {group=Ambient Occlusion, maskChannel=Blue, parameterName=Stain_Base_Channel, defaultValue=<rgba 0 0 1 0>}
  NamedRerouteDeclaration(None=@n11.R) {name=BaseLayerAO, nodeColor=<rgba 0.4531 1 0 1>}
  n12 = ChannelMaskParameter(None=@n10.XYZ Texture) {group=Ambient Occlusion, maskChannel=Green, parameterName=Stain_AO_Channel, defaultValue=<rgba 0 1 0 0>}
  NamedRerouteDeclaration(None=@n12.R) {name=Stain AO, nodeColor=<rgba 1 0 0.2344 1>}
  NamedRerouteUsage()
  Reroute(None=TextureSampleParameter2D() {parameterName=Normals, samplerType=SAMPLERTYPE_Normal, texture=&FlatNormalBlue}.RGB)
  TextureSampleParameter2D() {parameterName=Emissive, texture=&T_BlacksRGB}
 ]
}
